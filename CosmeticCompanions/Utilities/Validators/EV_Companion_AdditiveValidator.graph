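# Editor validator that checks additive animation settings on Cosmetic Companion assets before submit.
# It runs only for PreSubmit, on AnimSequences under /CosmeticCompanions/Assets whose name ends in _ADD.
# An embedded Python script fails sequences with no additive type, or a frame or scaled base pose lacking a reference sequence.
# Failures open a "Submit anyway?" dialog: Yes logs a warning and passes the asset, while declining marks it invalid.
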EditorUtilityBlueprint /CosmeticCompanions/Utilities/Validators/EV_Companion_AdditiveValidator
graphs [
 EventGraph {links [], nodes []}
 K2_CanValidate {nodes [
   n1 = Can Validate!() -> (then, InUsecase:EDataValidationUsecase Enum=None) {is=title K2Node_FunctionEntry}
   > n2 = Switch on EDataValidationUsecase!(execute, Selection:EDataValidationUsecase Enum=@n1) -> (None:Exec, Manual:Exec, Commandlet:Exec, Save:Exec, PreSubmit:Exec, Script:Exec) {is=title K2Node_SwitchEnum}
   Return Node!(execute=@n2.PreSubmit, ReturnValue:bool=true)
  ]}
 K2_CanValidateAsset {nodes [
   n1 = Can Validate Asset!() -> (then, InAsset:Object Reference) {is=title K2Node_FunctionEntry}
   n2 = Split(SourceString=GetObjectName(Object=@n1), InStr=_, SearchDir=FromEnd)
   n3 = Get Object Path String(Object=@n1)
   n4 = Get Class(Object=@n1) -> (ReturnValue:Object Class Reference)
   > Return Node(ReturnValue=AND Boolean!(
     A="Equal (Class)"(A=@n4, B=/Script/Engine.AnimSequence)
     B=Contains(SearchIn=@n3, Substring=/CosmeticCompanions/Assets)
     C:bool="Equal Exactly (String)"(A=@n2.RightS, B=ADD)
    ) -> (ReturnValue))
  ]}
 K2_ValidateLoadedAsset {nodes [
   n1 = Validate Loaded Asset!() -> (then, InAsset:Object Reference) {is=title K2Node_FunctionEntry}
   > n2 = Execute Python Script!(
    execute
    PythonScript="import unreal\r\n\r\nadditive_anim_type = anim_sequence.get_editor_property(\"additive_anim_type\")\r\n\r\ndef check_additive_settings(in_anim_sequence):\r\n\tif additive_anim_type == unreal.AdditiveAnimationType.AAT_NONE:\r\n\t\t##TODO: Throw Error\r\n\t\treturn False\r\n\telse:\r\n\t\t##TODO: check base type to see if it requires a reference pose and if so, check if there is a reference pose\r\n\t\tref_pose_type = in_anim_sequence.get_editor_property(\"ref_pose_type\")\r\n\t\tif ref_pose_type == unreal.AdditiveBasePoseType.ABPT_ANIM_FRAME or ref_pose_type == unreal.AdditiveBasePoseType.ABPT_ANIM_SCALED:\r\n\t\t\tref_pose = in_anim_sequence.get_editor_property(\"ref_pose_seq\")\r\n\t\t\tif not ref_pose:\r\n\t\t\t\treturn False\r\n\r\n\treturn True\r\n\r\n\r\npassed = check_additive_settings(anim_sequence)"
    anim_sequence:Object Reference=@n1
   ) -> (then, ReturnValue, passed:bool=~) {inputs [anim_sequence], outputs [passed]}
   > n3 = Branch(Condition=@n2.passed)
   AssetPasses!(execute=[@n7.then, @n3.then], self:Editor Validator Base Object Reference, InAsset:Object Reference=@n1) -> (then)
   > Return Node(ReturnValue:EDataValidationResult Enum=Valid)
   n4 = Show Message Dialog(
    execute=@n3.else
    Title=MakeLiteralText(Value="LOCTEXT(\"[WARNING] Missing Additive Setting\")")
    Message=Format Text!(
     Format="LOCTEXT(\"Naming indicates {AssetName} is an additive anim but it is missing additive settings. Please verify your animation is not breaking in game. Submit anyway?\")"
     AssetName:String=GetDisplayName(Object=@n1)
    ) -> (Result)
    MessageType=YesNo
    DefaultValue=No
    MessageCategory=Warning
   )
   > n5 = Switch on EAppReturnType(Selection=@n4)
   n6 = MakeLiteralText(Value="LOCTEXT(\"Missing Additive Setting\")")
   > AssetFails!(execute, self:Editor Validator Base Object Reference, InAsset:Object Reference=@n1, InMessage:"Text (by ref)"=@n6) -> (then)
   > Return Node(ReturnValue=Invalid)
   n7 = AssetWarning!(execute=@n5.Yes, self:Editor Validator Base Object Reference, InAsset:Object Reference=@n1, InMessage:"Text (by ref)"=@n6) -> (then)
  ]}
]
variables []
